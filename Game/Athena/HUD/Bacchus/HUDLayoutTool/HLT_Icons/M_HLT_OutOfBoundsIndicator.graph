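# Translucent UI material for the HUD Layout Tool's out-of-bounds indicator icon, drawing a procedural X without textures.
# UVs are rotated around the center onto each diagonal, and each rotated X is multiplied by its inverse, peaking along that diagonal.
# SmoothStep, with Stroke Width as both min and max, cuts each peak into a hard thin line.
# The two lines are added together, and a Color parameter defaults to white.
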Material /Game/Athena/HUD/Bacchus/HUDLayoutTool/HLT_Icons/M_HLT_OutOfBoundsIndicator
graph {
 properties {blendMode=BLEND_Translucent, materialDomain=MD_UI, pixelDepthOffsetMode=PDOM_Legacy, usedShadingModels=Default Lit}
 nodes [
  n1 = MaterialFunctionCall(UVs=TextureCoordinate(), Rotation Center=Constant() {r=0.5}, "Rotation Angle (0-1)"=Constant() {r=0.125}) {materialFunction=&CustomRotator}
  n2 = OneMinus(None=ComponentMask(None=@n1.Rotated Values) {r=true})
  n3 = ComponentMask(None=@n1.Rotated Values) {r=true}
  n4 = ScalarParameter() {defaultValue=0.2499, parameterName=Stroke Width, sliderMax=0.2499, sliderMin=0.2495}
  n5 = MaterialFunctionCall(UVs=TextureCoordinate(), Rotation Center=Constant() {r=0.5}, "Rotation Angle (0-1)"=Constant() {r=0.375}) {materialFunction=&CustomRotator}
  n6 = OneMinus(None=ComponentMask(None=@n5.Rotated Values) {r=true})
  n7 = ComponentMask(None=@n5.Rotated Values) {r=true}
  Add(
   A=MaterialFunctionCall(Alpha=Multiply(A=@n2, B=@n3), Min=@n4, Max=@n4) {materialFunction=&SmoothStep}.Result
   B=MaterialFunctionCall(Alpha=Multiply(A=@n6, B=@n7), Min=@n4, Max=@n4) {materialFunction=&SmoothStep}.Result
  )
  VectorParameter() {parameterName=Color, defaultValue=<rgba 1 1 1 1>}
 ]
}
