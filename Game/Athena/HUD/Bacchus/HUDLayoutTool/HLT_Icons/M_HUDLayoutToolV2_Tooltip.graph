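# Translucent UI material that procedurally draws a HUD Layout Tool tooltip bubble and stem, with a selection highlight.
# Rotated, offset UV gradients form bubble edges and an angled stem, scaled by Sharpness into shape and stroke-inset masks.
# The inset bubble mask picks between outer and inner stroke colors, and AnimateSelection blends these toward selected colors.
# A static switch, ShineAnimationOn, holds a color-dodge shine swipe.
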Material /Game/Athena/HUD/Bacchus/HUDLayoutTool/HLT_Icons/M_HUDLayoutToolV2_Tooltip
graph {
 properties {blendMode=BLEND_Translucent, materialDomain=MD_UI, pixelDepthOffsetMode=PDOM_Legacy, usedShadingModels=Default Lit}
 nodes [
  n1 = VectorParameter() {desc=Outer Stroke Color, group=Stroke, parameterName=OuterStrokeColor, defaultValue=<rgba 0 0.0908 1 1>}
  n2 = VectorParameter() {desc=Inner Stroke Color, group=Stroke, parameterName=InnerStrokeColor, defaultValue=<rgba 0 0 0.3005 1>}
  n3 = ScalarParameter() {defaultValue=32, group=Feather, parameterName=Sharpness}
  n4 = TextureCoordinate() {coordinateIndex=4}
  n5 = Rotator(Coordinate=@n4, Time=Divide(A=ScalarParameter() {group=Degrees, parameterName=BubbleTopAngle}) {constB=360}) {speed=6.2831}
  n6 = Multiply(A=Subtract(
    A=ComponentMask(None=@n5) {desc=Only allow Green channel, g=true}
    B=ScalarParameter() {defaultValue=0.02, group=Offset, parameterName=BubbleTopOffset}
   ) {constB=0}) {constB=0.75}
  n7 = TextureCoordinate()
  n8 = Rotator(Coordinate=@n7, Time=Divide(A=ScalarParameter() {group=Degrees, parameterName=BubbleLeftAngle, sliderMax=360}) {constB=360}) {speed=6.2831}
  n9 = Rotator(Coordinate=@n7, Time=Divide(A=ScalarParameter() {defaultValue=180, group=Degrees, parameterName=BubbleRightAngle}) {constB=360}) {speed=6.2831}
  n10 = DotProduct(
   A=Subtract(A=ComponentMask(None=@n8) {r=true}, B=ScalarParameter() {defaultValue=0.02, group=Offset, parameterName=BubbleLeftOffset})
   B=Subtract(A=ComponentMask(None=@n9) {r=true}, B=ScalarParameter() {defaultValue=0.02, group=Offset, parameterName=BubbleRightOffset})
  )
  n11 = Multiply(A=AppendVector(A=Saturate(None=@n6), B=Saturate(None=@n10))) {constB=4}
  n12 = ScalarParameter() {defaultValue=0.125, parameterName=StrokeThickness}
  n13 = Saturate(None=Multiply(A=@n3, B=Subtract(A=@n11, B=@n12)))
  n14 = ComponentMask(None=@n13) {g=true}
  n15 = ComponentMask(None=@n13) {r=true}
  n16 = Rotator(Coordinate=@n4, Time=Divide(A=ScalarParameter() {defaultValue=180, group=Degrees, parameterName=BubbleBottomAngle}) {constB=360}) {speed=6.2831}
  n17 = Multiply(A=Subtract(
    A=ComponentMask(None=@n16) {desc=Only allow Green channel, g=true}
    B=ScalarParameter() {defaultValue=0.02, group=Offset, parameterName=BubbleBottomOffset, sliderMax=0.5}
   ) {constB=0}) {constB=0.75}
  n18 = Multiply(A=AppendVector(A=Saturate(None=@n17), B=Constant())) {constB=4}
  n19 = Multiply(A=Subtract(A=@n18, B=@n12), B=@n3)
  n20 = LinearInterpolate(B=Saturate(None=DotProduct(A=@n14, B=@n15)), Alpha=Saturate(None=ComponentMask(None=@n19) {r=true}))
  n21 = VectorParameter() {desc=Outer Stroke Color, group=Stroke, parameterName=OuterStrokeColor_Selected, defaultValue=<rgba 0 0.8388 1 1>}
  n22 = VectorParameter() {desc=Inner Stroke Color, group=Stroke, parameterName=InnerStrokeColor_Selected, defaultValue=<rgba 0.0091 0.1248 0.7011 1>}
  LinearInterpolate(
   A=LinearInterpolate(A=@n1.RGB, B=@n2.RGB, Alpha=@n20)
   B=LinearInterpolate(A=@n21.RGB, B=@n22.RGB, Alpha=@n20)
   Alpha=ScalarParameter() {parameterName=AnimateSelection, sliderMax=1}
  )
  n23 = TextureCoordinate()
  n24 = ScalarParameter() {defaultValue=0.5, group=Offset, parameterName=StemXRotationPivot, sliderMax=1}
  n25 = AppendVector(A=@n24, B=Constant() {r=1})
  n26 = MaterialFunctionCall(
   UVs=@n23
   Rotation Center=@n25
   "Rotation Angle (0-1)"=Divide(A=ScalarParameter() {group=Degrees, parameterName=StemRightSideAngle, sliderMax=180}) {constB=360}
  ) {materialFunction=&CustomRotator}
  n27 = AppendVector(A=Multiply(A=@n24) {constB=-1}, B=Constant() {r=-1})
  n28 = ScalarParameter() {defaultValue=126, group=Degrees, parameterName=StemLeftSideAngle, sliderMax=180}
  n29 = MaterialFunctionCall(UVs=@n23, Rotation Center=@n25, "Rotation Angle (0-1)"=Divide(A=Multiply(A=@n28) {constB=-1}) {constB=360}) {materialFunction=&CustomRotator}
  n30 = AppendVector(A=ComponentMask(None=Add(A=@n26.Rotated Values, B=@n27)) {r=true}, B=ComponentMask(None=Add(A=@n29.Rotated Values, B=@n27)) {g=true})
  n31 = Divide(A=ScalarParameter() {group=Stroke, parameterName=StrokeThickness, sliderMax=0.7}) {constB=3}
  n32 = Multiply(A=ScalarParameter() {defaultValue=32, group=Feather, parameterName=Sharpness}) {constB=10}
  n33 = Multiply(A=Add(A=@n30, B=Multiply(A=@n31) {constB=-1}), B=@n32)
  LinearInterpolate(B=Saturate(None=ComponentMask(None=@n33) {r=true}), Alpha=Saturate(None=ComponentMask(None=@n33) {g=true}))
  n34 = Multiply(A=@n30, B=@n32)
  LinearInterpolate(B=Saturate(None=ComponentMask(None=@n34) {r=true}), Alpha=Saturate(None=ComponentMask(None=@n34) {g=true}))
  n35 = Saturate(None=Multiply(A=@n11, B=@n3))
  n36 = ComponentMask(None=@n35) {g=true}
  n37 = ComponentMask(None=@n35) {r=true}
  n38 = Multiply(A=@n18, B=@n3)
  LinearInterpolate(B=Saturate(None=DotProduct(A=@n36, B=@n37)), Alpha=Saturate(None=ComponentMask(None=@n38) {r=true}))
  StaticSwitchParameter(True=MaterialFunctionCall!(Blend=MaterialFunctionCall!(In_Uv) {materialFunction=/Game/UI/Foundation/Materials/M-Function-ShineSwipe}.Result, Base) {materialFunction=/Engine/Functions/Engine_MaterialFunctions03/Blends/Blend_ColorDodge}.Result) {parameterName=ShineAnimationOn}
 ]
}
