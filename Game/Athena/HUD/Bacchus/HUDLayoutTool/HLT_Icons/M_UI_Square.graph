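# M_UI_Square is a translucent UI material for the HUD Layout Tool icons that draws a square outline with optional animated hatching.
# Inverting the mask from center-scaled UVs leaves a procedural border sized by OutlineWidth.
# A static switch subtracts stripes from that border, built from rotated, panned UVs through sine and ceil, with adjustable angle, speed, offset and scale.
# A Color vector parameter is exposed for tinting.
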Material /Game/Athena/HUD/Bacchus/HUDLayoutTool/HLT_Icons/M_UI_Square
graph {
 properties {blendMode=BLEND_Translucent, materialDomain=MD_UI, pixelDepthOffsetMode=PDOM_Legacy, usedShadingModels=Default Lit}
 nodes [
  n1 = ScalarParameter() {defaultValue=0.2, parameterName=OutlineWidth, sliderMax=1}
  n2 = OneMinus(None=MaterialFunctionCall(UVs=TextureCoordinate(), Texture Scale=OneMinus(None=@n1)) {materialFunction=&ScaleUVsByCenter}.0-1 mask)
  n3 = TextureCoordinate()
  n4 = MaterialFunctionCall(Multiplier=ScalarParameter() {defaultValue=1, parameterName=Hatching Angle}) {materialFunction=&Pi}
  n5 = ScalarParameter() {defaultValue=0.5, parameterName=Speed X}
  n6 = Constant()
  n7 = Add(
   A=Panner(Coordinate=Rotator(Coordinate=@n3, Time=@n4.Result), Speed=AppendVector(A=@n5, B=@n6))
   B=ScalarParameter() {parameterName=Hatching Offset, sliderMax=1}
  )
  n8 = ScalarParameter() {defaultValue=0.2, parameterName=Hatching Scale, sliderMax=1}
  n9 = Sine(None=ComponentMask(None=Divide(A=@n7, B=@n8) {constB=0.01}) {r=true})
  StaticSwitchParameter(True=Subtract(A=@n2, B=Ceil(None=@n9)), False=@n2) {parameterName=Hatching ON/OFF}
  VectorParameter() {parameterName=Color, defaultValue=<rgba 1 1 1 1>}
 ]
}
